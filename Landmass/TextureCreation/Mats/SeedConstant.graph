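# Unlit, alpha-composite material in Landmass texture creation, apparently a seed-constant scratch graph for inspecting packed height values.
# It feeds a constant height into PackRG8Height, then sends the XY result to a float2 debug display and a small subtraction.
# Separate branches display vector and scalar parameters through float4 debug functions and build a mid-grey constant with MakeFloat3.
# The graph shows no branch wired to a material output.
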Material /Landmass/TextureCreation/Mats/SeedConstant
graph {
 properties {blendMode=BLEND_AlphaComposite, pixelDepthOffsetMode=PDOM_Legacy, shadingModel=MSM_Unlit, usedShadingModels=Unlit}
 nodes [
  n1 = VectorParameter() {parameterName=asdfsaf, defaultValue=<rgba 1 2 3 4>}
  ComponentMask(None=MaterialFunctionCall!(Vector4=AppendVector(A=@n1.RGB, B=@n1.RGB), MaximumNumberOfDigits, UVs, DebugTextLocation RG_UpperRight BA_LowerLeft, Component Spacing) {materialFunction=/Engine/Functions/Engine_MaterialFunctions02/Utility/DebugFloat4Values}.ColorCodedOutput) {b=true, g=true, r=true}
  MaterialFunctionCall(X=Divide(A=Constant() {r=128}) {constB=255}) {materialFunction=&MakeFloat3}
  n2 = MaterialFunctionCall!(Height Int -32k to 32k=Constant()) {materialFunction=/Landmass/Landscape/OldPrototype_BP/Mats/MF/PackRG8Height}
  MaterialFunctionCall(Float4=MaterialFunctionCall!(Vector2=@n2.XY, MaximumNumberOfDigits, UVs, DebugTextLocation RG_UpperRight BA_LowerLeft, Component Spacing) {materialFunction=/Engine/Functions/Engine_MaterialFunctions02/Utility/DebugFloat2Values}.ColorCodedOutput) {materialFunction=&BreakOutFloat4Components}
  OneMinus(None=MaterialFunctionCall(Vector4=ScalarParameter() {defaultValue=999, parameterName=999}) {materialFunction=&DebugFloat4Values}.GreyScaleOutput)
  Subtract(A=@n2.XY) {constB=0.0001}
 ]
}
